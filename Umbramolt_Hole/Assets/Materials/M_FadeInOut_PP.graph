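# M_FadeInOut_PP is an unlit post-process material that fades the screen to or from black.
# Time minus TimeStart, divided by Duration and clamped, gives progress, which a saturated TimeStart multiplies to hold it at zero while unset.
# FadeInOutFlag blends between that progress and its inverse to pick the fade direction.
# The result is smoothstepped and routed through the TransitionScalar named reroute to lerp scene color toward black.
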Material /Umbramolt_Hole/Assets/Materials/M_FadeInOut_PP
graph {
 properties {materialDomain=MD_PostProcess, pixelDepthOffsetMode=PDOM_Legacy, shadingModel=MSM_Unlit, usedShadingModels=Unlit}
 nodes [
  LinearInterpolate(A=ComponentMask(None=SceneTexture() {sceneTextureId=PPI_PostProcessInput0}.Color) {b=true, g=true, r=true}, Alpha=NamedRerouteUsage()) {constB=0}
  n1 = ScalarParameter() {parameterName=TimeStart}
  n2 = Subtract(A=Time(), B=@n1)
  n3 = ScalarParameter() {defaultValue=1.2, parameterName=Duration}
  n4 = Multiply(A=Saturate(None=Divide(A=@n2, B=@n3)), B=Saturate(None=@n1))
  n5 = OneMinus(None=@n4)
  n6 = ScalarParameter() {desc="0 = Fade from screen to black; 1 = Fade from black to screen;", parameterName=FadeInOutFlag}
  NamedRerouteDeclaration(None=SmoothStep(Value=LinearInterpolate(A=@n4, B=@n5, Alpha=@n6))) {name=TransitionScalar, nodeColor=<rgba 0.8047 1 0 1>}
 ]
}
